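# Animation Blueprint for the CowboyPistol revolver weapon skeleton that steps the chamber through six rotation positions.
# A non-target base pose passes through DefaultSlot, then the Chamber Rotation state machine adds one-shot additive animations on top.
# The ChamberRotation notify increments a counter that latches per-position booleans, which drive the cyclic state transitions.
# Position six resets the counter and clears the earlier flags; position one clears the sixth.
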AnimBlueprint /Game/Animation/Game/MainPlayer/Combat/Guns/CowboyPistol/CowboyPistol_AnimBP
target_skeleton=/Game/Weapons/Skeleton_Templates/SK_Revolver_SideAction_Skeleton
graphs [
 AnimGraph {nodes [
   n1 = Sequence Player 'CowboyPistol_NonTarget_Pose_CMM_Weapon'() {sequence=&/CowboyPistol_NonTarget_Pose_CMM_Weapon}
   AnimGraphNode_StateMachine_0 = Chamber Rotation() {is=title AnimGraphNode_StateMachine}
   Output Pose(Result=Apply Additive(Base=Slot 'DefaultSlot'(Source=@n1), Additive=@AnimGraphNode_StateMachine_0) {alphaBoolBlend {blendOption=Linear}}) {layerGroup=None}
  ]}
 Chamber Rotation {
  entry=Default Rotation
  path=AnimGraph.AnimGraphNode_StateMachine_0.Chamber Rotation
  ids=id
  nodes [
   AnimStateNode_0 = Default Rotation() {is=title AnimStateNode}
   AnimStateNode_1 = ChamberPosition 1() {is=title AnimStateNode}
   AnimStateNode_2 = ChamberPosition 2() {is=title AnimStateNode}
   AnimStateNode_3 = ChamberPosition 3() {is=title AnimStateNode}
   AnimStateNode_4 = ChamberPosition 4() {is=title AnimStateNode}
   AnimStateNode_5 = ChamberPosition 5() {is=title AnimStateNode}
   AnimStateNode_6 = ChamberPosition 6() {is=title AnimStateNode}
   AnimStateTransitionNode_0 = AnimStateTransitionNode() {crossfadeDuration=0.033}
   AnimStateTransitionNode_1 = AnimStateTransitionNode() {crossfadeDuration=0.033}
   AnimStateTransitionNode_2 = AnimStateTransitionNode() {crossfadeDuration=0.033}
   AnimStateTransitionNode_3 = AnimStateTransitionNode() {crossfadeDuration=0.033}
   AnimStateTransitionNode_4 = AnimStateTransitionNode() {crossfadeDuration=0.033}
   AnimStateTransitionNode_5 = AnimStateTransitionNode() {crossfadeDuration=0.033}
   AnimStateTransitionNode_6 = AnimStateTransitionNode() {crossfadeDuration=0.033}
  ]
  links [
   {
    dst=ChamberPosition 2
    rule=AnimGraph.AnimGraphNode_StateMachine_0.Chamber Rotation.AnimStateTransitionNode_1.Transition
    src=ChamberPosition 1
    via=AnimStateTransitionNode_1
   }
   {
    dst=ChamberPosition 3
    rule=AnimGraph.AnimGraphNode_StateMachine_0.Chamber Rotation.AnimStateTransitionNode_2.Transition
    src=ChamberPosition 2
    via=AnimStateTransitionNode_2
   }
   {
    dst=ChamberPosition 4
    rule=AnimGraph.AnimGraphNode_StateMachine_0.Chamber Rotation.AnimStateTransitionNode_3.Transition
    src=ChamberPosition 3
    via=AnimStateTransitionNode_3
   }
   {
    dst=ChamberPosition 5
    rule=AnimGraph.AnimGraphNode_StateMachine_0.Chamber Rotation.AnimStateTransitionNode_4.Transition
    src=ChamberPosition 4
    via=AnimStateTransitionNode_4
   }
   {
    dst=ChamberPosition 6
    rule=AnimGraph.AnimGraphNode_StateMachine_0.Chamber Rotation.AnimStateTransitionNode_5.Transition
    src=ChamberPosition 5
    via=AnimStateTransitionNode_5
   }
   {
    dst=ChamberPosition 1
    rule=AnimGraph.AnimGraphNode_StateMachine_0.Chamber Rotation.AnimStateTransitionNode_6.Transition
    src=ChamberPosition 6
    via=AnimStateTransitionNode_6
   }
   {
    dst=ChamberPosition 1
    rule=AnimGraph.AnimGraphNode_StateMachine_0.Chamber Rotation.AnimStateTransitionNode_0.Transition
    src=Default Rotation
    via=AnimStateTransitionNode_0
   }
  ]
 }
 Default Rotation {
  path=AnimGraph.AnimGraphNode_StateMachine_0.Chamber Rotation.AnimStateNode_0.Default Rotation
  nodes [Output Animation Pose(Result=Additive Identity Pose())]
 }
 ChamberPosition 1 {
  path=AnimGraph.AnimGraphNode_StateMachine_0.Chamber Rotation.AnimStateNode_1.ChamberPosition 1
  nodes [Output Animation Pose(Result="Sequence Player (additive) 'CowboyPistol_ChamberRot_1_Weapon'"() {bLoopAnimation=false, sequence=&/CowboyPistol_ChamberRot_1_Weapon})]
 }
 ChamberPosition 2 {
  path=AnimGraph.AnimGraphNode_StateMachine_0.Chamber Rotation.AnimStateNode_2.ChamberPosition 2
  nodes [Output Animation Pose(Result="Sequence Player (additive) 'CowboyPistol_ChamberRot_2_Weapon'"() {bLoopAnimation=false, sequence=&/CowboyPistol_ChamberRot_2_Weapon})]
 }
 ChamberPosition 3 {
  path=AnimGraph.AnimGraphNode_StateMachine_0.Chamber Rotation.AnimStateNode_3.ChamberPosition 3
  nodes [Output Animation Pose(Result="Sequence Player (additive) 'CowboyPistol_ChamberRot_3_Weapon'"() {bLoopAnimation=false, sequence=&/CowboyPistol_ChamberRot_3_Weapon})]
 }
 ChamberPosition 4 {
  path=AnimGraph.AnimGraphNode_StateMachine_0.Chamber Rotation.AnimStateNode_4.ChamberPosition 4
  nodes [Output Animation Pose(Result="Sequence Player (additive) 'CowboyPistol_ChamberRot_4_Weapon'"() {bLoopAnimation=false, sequence=&/CowboyPistol_ChamberRot_4_Weapon})]
 }
 ChamberPosition 5 {
  path=AnimGraph.AnimGraphNode_StateMachine_0.Chamber Rotation.AnimStateNode_5.ChamberPosition 5
  nodes [Output Animation Pose(Result="Sequence Player (additive) 'CowboyPistol_ChamberRot_5_Weapon'"() {bLoopAnimation=false, sequence=&/CowboyPistol_ChamberRot_5_Weapon})]
 }
 ChamberPosition 6 {
  path=AnimGraph.AnimGraphNode_StateMachine_0.Chamber Rotation.AnimStateNode_6.ChamberPosition 6
  nodes [Output Animation Pose(Result="Sequence Player (additive) 'CowboyPistol_ChamberRot_6_Weapon'"() {bLoopAnimation=false, sequence=&/CowboyPistol_ChamberRot_6_Weapon})]
 }
 Transition {
  path=AnimGraph.AnimGraphNode_StateMachine_0.Chamber Rotation.AnimStateTransitionNode_0.Transition
  nodes [Result(bCanEnterTransition=Get ChamberPosition_1() -> (ChamberPosition_1:bool))]
 }
 Transition {
  path=AnimGraph.AnimGraphNode_StateMachine_0.Chamber Rotation.AnimStateTransitionNode_1.Transition
  nodes [Result(bCanEnterTransition=Get ChamberPosition_2() -> (ChamberPosition_2:bool))]
 }
 Transition {
  path=AnimGraph.AnimGraphNode_StateMachine_0.Chamber Rotation.AnimStateTransitionNode_2.Transition
  nodes [Result(bCanEnterTransition=Get ChamberPosition_3() -> (ChamberPosition_3:bool))]
 }
 Transition {
  path=AnimGraph.AnimGraphNode_StateMachine_0.Chamber Rotation.AnimStateTransitionNode_3.Transition
  nodes [Result(bCanEnterTransition=Get ChamberPosition_4() -> (ChamberPosition_4:bool))]
 }
 Transition {
  path=AnimGraph.AnimGraphNode_StateMachine_0.Chamber Rotation.AnimStateTransitionNode_4.Transition
  nodes [Result(bCanEnterTransition=Get ChamberPosition_5() -> (ChamberPosition_5:bool))]
 }
 Transition {
  path=AnimGraph.AnimGraphNode_StateMachine_0.Chamber Rotation.AnimStateTransitionNode_5.Transition
  nodes [Result(bCanEnterTransition=Get ChamberPosition_6() -> (ChamberPosition_6:bool))]
 }
 Transition {
  path=AnimGraph.AnimGraphNode_StateMachine_0.Chamber Rotation.AnimStateTransitionNode_6.Transition
  nodes [Result(bCanEnterTransition=Get ChamberPosition_1())]
 }
 EventGraph {nodes [
   AnimNotify_ChamberRotation() {is=title K2Node_Event}
   > n1 = Increment Int(Value=Get ChamberRotationCounter() -> (ChamberRotationCounter:int=0))
   > n2 = Switch on Int!(execute:Exec, Selection:int=@n1) -> (Default:Exec, 0:Exec, 1:Exec, 2:Exec, 3:Exec, 4:Exec, 5:Exec, 6:Exec)
   Set ChamberPosition_1(execute=[@n2.1, @n2.Default], ChamberPosition_1:bool=true) -> (Output_Get:bool)
   > Set ChamberPosition_6(ChamberPosition_6:bool=false) -> (Output_Get:bool)
   Set ChamberPosition_2(execute=@n2.2, ChamberPosition_2:bool=true) -> (Output_Get:bool)
   Set ChamberPosition_3(execute=@n2.3, ChamberPosition_3:bool=true) -> (Output_Get:bool)
   Set ChamberPosition_4(execute=@n2.4, ChamberPosition_4:bool=true) -> (Output_Get:bool)
   Set ChamberPosition_5(execute=@n2.5, ChamberPosition_5:bool=true) -> (Output_Get:bool)
   Set ChamberPosition_6(execute=@n2.6, ChamberPosition_6=true)
   > Set ChamberRotationCounter(ChamberRotationCounter:int=0) -> (Output_Get:int=0)
   > Set ChamberPosition_1(ChamberPosition_1=false)
   > Set ChamberPosition_2(ChamberPosition_2=false)
   > Set ChamberPosition_3(ChamberPosition_3=false)
   > Set ChamberPosition_4(ChamberPosition_4=false)
   > Set ChamberPosition_5(ChamberPosition_5=false)
   TryGetPawnOwner()
   Event BlueprintUpdateAnimation()
  ]}
]
variables [
 ChamberPosition_1 {type=boolean}
 ChamberPosition_2 {type=boolean}
 ChamberPosition_3 {type=boolean}
 ChamberPosition_4 {type=boolean}
 ChamberPosition_5 {type=boolean}
 ChamberPosition_6 {type=boolean}
 ChamberRotationCounter {type=integer, default=1}
]
